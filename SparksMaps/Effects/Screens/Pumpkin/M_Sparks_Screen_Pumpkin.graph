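# Unlit material for a Sparks stage screen showing an animated pumpkin face with flame, smoky background and radial glow.
# Channels of one pumpkin texture supply the layers, with panning noise distorting the flame and a slow sine pulsing UV scale.
# Stage colors from MF_MojoStageColors and a beat-driven curve atlas reactivity value are shared through named reroutes.
# The summed layers feed MF_Sparks_Screen_Details as emissive for the screen finish.
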Material /SparksMaps/Effects/Screens/Pumpkin/M_Sparks_Screen_Pumpkin
graph {
 properties {
  bAutomaticallySetUsageInEditor=false
  bHasPixelAnimation=true
  bUsedWithNanite=true
  pixelDepthOffsetMode=PDOM_Legacy
  shadingModel=MSM_Unlit
  usedShadingModels=Unlit
 }
 nodes [
  Constant() {r=-0.01}
  n1 = NamedRerouteUsage()
  n2 = Constant() {r=0.5}
  n3 = Constant() {r=0.5}
  n4 = Multiply(
   A=Multiply(
    A=MaterialFunctionCall(UVs=@n1, Radius=@n2, Density=@n3) {materialFunction=&RadialGradientExponential}.RadialGradientExponential
    B=NamedRerouteUsage()
   )
   B=NamedRerouteUsage()
  )
  n5 = MaterialFunctionCall(
   Input=NamedRerouteUsage()
   Input Low=Constant()
   Input High=Constant() {r=1}
   Target Low=Constant() {r=0.75}
   Target High=Constant() {r=1.25}
  ) {materialFunction=&RemapValueRange}
  n6 = NamedRerouteUsage()
  n7 = TextureCoordinate()
  n8 = VectorParameter() {parameterName=UV Scale and Scale By Center, defaultValue=<rgba 0 -0.1 0.75 1>}
  n9 = AppendVector(A=@n8.R, B=@n8.R)
  n10 = AppendVector(A=@n8.B, B=@n8.B)
  n11 = Multiply(A=Time()) {constB=0.25}
  n12 = MaterialFunctionCall(
   Input=ConstantBiasScale(None=Sine(None=@n11))
   Input Low=Constant()
   Input High=Constant() {r=1}
   Target Low=Constant() {r=1}
   Target High=Constant() {r=1.15}
  ) {materialFunction=&RemapValueRange}
  n13 = NamedRerouteDeclaration(None=MaterialFunctionCall(UVs=Add(A=@n7, B=@n9), Texture Scale=Multiply(A=@n10, B=@n12.Result)) {materialFunction=&ScaleUVsByCenter}.UVs) {name=UVs, nodeColor=<rgba 0.6641 0 1 1>}
  n14 = TextureCoordinate()
  n15 = VectorParameter() {parameterName=Flame UVs and Panner, defaultValue=<rgba 2 1 -0.1 0.4>}
  n16 = AppendVector(A=@n15.R, B=@n15.R)
  n17 = TextureSample(UVs=Panner(Coordinate=TextureCoordinate()) {speedX=0.1, speedY=0.1}) {texture=/Game/Effects/Fort_Effects/Textures/T_Smooth_tile}
  n18 = LinearInterpolate(A=Panner(Coordinate=Multiply(A=@n14, B=@n16), Speed=AppendVector(A=@n15.B, B=@n15.B)), B=@n17.G) {constAlpha=0.05}
  n19 = TextureSample(UVs=LinearInterpolate(A=Add(A=@n13) {constB=0.05}, B=TextureSample(UVs=@n18) {texture=/Game/Effects/Fort_Effects/Textures/Tile/T_Inky_Smoke_Tile}.G) {constAlpha=0.1}) {samplerType=SAMPLERTYPE_LinearColor, texture=&/T_Sparks_Screen_Pumpkin}
  n20 = Constant() {r=0.7}
  n21 = Multiply(A=Power(Base=Multiply(A=@n19.G, B=@n20) {constB=0.25}) {constExponent=2.5}, B=NamedRerouteUsage())
  n22 = NamedRerouteUsage()
  n23 = Multiply(A=MaterialFunctionCall() {materialFunction=&LinearGradient}.VGradient, B=@n22)
  n24 = NamedRerouteUsage()
  n25 = TextureSample(UVs=@n24) {samplerType=SAMPLERTYPE_LinearColor, texture=&T_Sparks_Screen_Pumpkin}
  n26 = Power(Base=@n25.R)
  n27 = Power(Base=@n25.B, Exp=Constant() {r=0.1}) {constExponent=0.5}
  n28 = Add(A=Multiply(A=@n26, B=Constant() {r=0.025}), B=Multiply(A=Power(Base=@n25.R), B=OneMinus(None=@n27)))
  n29 = Multiply(A=NamedRerouteUsage(), B=Constant() {r=-0.01}) {constB=0.1}
  n30 = Power(Base=TextureSample(UVs=Add(A=@n29, B=@n24) {constB=0.1}) {samplerType=SAMPLERTYPE_LinearColor, texture=&T_Sparks_Screen_Pumpkin}.R)
  n31 = NamedRerouteDeclaration(None=OneMinus(None=@n26)) {name=FaceMask, nodeColor=<rgba 1 0 0.9141 1>}
  n32 = Add(
   A=Add(A=Multiply(A=@n6, B=@n21), B=Multiply(A=@n28, B=@n23))
   B=Multiply(A=Multiply(A=@n30, B=@n31), B=Constant3Vector() {constant=<rgba 0.5 0.0006 0.0006 1>})
  )
  n33 = TextureCoordinate()
  n34 = VectorParameter() {parameterName=BG Smoke UVs and Panner, defaultValue=<rgba 2 1 0 0.2>}
  n35 = AppendVector(A=@n34.R, B=@n34.R)
  n36 = LinearInterpolate(A=Panner(Coordinate=Multiply(A=@n33, B=@n35), Speed=AppendVector(A=@n34.B, B=@n34.B)), B=@n17.G, Alpha=Constant() {r=0.2}) {constAlpha=0.1}
  n37 = MaterialFunctionCall() {materialFunction=&LinearGradient}
  n38 = Constant() {r=5}
  n39 = Add(
   A=Multiply(A=TextureSample(UVs=@n36) {texture=/Game/Effects/Fort_Effects/Textures/Tile/T_SpookyFaces_Chroma_Tile}.G, B=Constant() {r=0.5})
   B=Multiply(A=Power(Base=@n37.VGradient, Exp=@n38), B=Constant() {r=0.5})
  )
  n40 = Power(
   Base=MaterialFunctionCall() {materialFunction=&LinearGradient}.VGradient
   Exp=Constant() {r=4}
  ) {constExponent=5}
  n41 = Multiply(A=@n22, B=Multiply(A=NamedRerouteUsage(), B=Multiply(A=@n39, B=@n40)))
  MaterialFunctionCall!(Emissive=Add(A=Multiply(A=@n4, B=@n5.Result), B=Add(A=@n32, B=@n41)), Scanlines Tile) {materialFunction=/SparksMaps/Effects/Screens/Materials/MF/MF_Sparks_Screen_Details}
  n42 = MaterialFunctionCall!(CustomUPosition) {materialFunction=/SparksStageShow/Systems/StageManager/FXManager/MaterialFunctions/MF_MojoStageColors}
  NamedRerouteDeclaration(None=@n42.Primary) {name=PrimaryColor, nodeColor=<rgba 0 0 0 1>}
  n43 = CurveAtlasRowParameter(CurveTime=CollectionParameter() {collection=/FMAudioReactiveData/AudioAnalysis/MPC_Sparks_Audio_Parameters, parameterName=BeatLifetime}) {
   atlas=/SparksMaps/Effects/Screens/Curves/CurveAtlas_Audio_Beat_A
   curve=/SparksMaps/Effects/Screens/Curves/Curve_Audio_Beat_Remap_001
   parameterName=ReactivityCurve
  }
  NamedRerouteDeclaration(None=@n43.R) {name=Reactivity, nodeColor=<rgba 0 0 0 1>}
  NamedRerouteDeclaration(None=@n42.Secondary) {name=SecondaryColor, nodeColor=<rgba 0 0 0 1>}
  NamedRerouteDeclaration(None=StaticSwitchParameter(True=NamedRerouteUsage(), False=NamedRerouteUsage()) {parameterName=IsSecondaryColor}) {name=Color, nodeColor=<rgba 0 0 0 1>}
 ]
}
